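; Gameplay ability that makes an NPC play a speech bubble bark after being rescued from a loot sac.
; On activation it stores the owning pawn, or ends if invalid, then waits for a loot sac rescue gameplay event on that pawn.
; When the event arrives, it waits briefly, plays a speech bubble for the rescue bark tag through a shared AI function library, and ends.
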(event Ability|EventActivateAbility
  (bind _returnvalue (Ability|GetActivatingPawn))
  ;[0 0 0] Activate Ability
  (|CallParentFunction)
  (Utilities|IsValid _returnvalue
    (:Is%20Valid
      (Variables|OwnerInfo|SetOwningPawn _returnvalue)
      (CallFunction|StartListeningtoLootSacRescue))
    (:Is%20Not%20Valid
      (Ability|EndAbility))))

(event Custom|PlayRescueBarks
  ;;[0.1144 0.1144 0.1144] Base Complete
  ;[0 0 0] Loot Sac Rescue
  (Utilities|FlowControl|Delay 2.0)
  (CallFunction|PlayRandomBarkforTag (Variables|Barks|GetRescueBarkTag))
  (Ability|EndAbility))

(event Custom|PlayRandomBarkForTag
  (params (BarkTag :type Gameplay%20Tag%20Structure) (bDelay :type Boolean))
  ;[0 0 0] Helper events
  (if bDelay
    (Utilities|FlowControl|Delay (Math|Random|RandomFloatinRange 1.0 5.0))
    (else
      _))
  (Class|BPFLAICommonFunctions|PlaySpeechBubble (Variables|OwnerInfo|GetOwningPawn) BarkTag))

(event Custom|StartListeningToLootSacRescue
  ;[0 0 0] Helper events
  ;;[0.1144 0.1144 0.1144] Wait for the NPC to be rescued from the loot sac
  (Ability|Async|WaitGameplayEventToActor (Variables|OwnerInfo|GetOwningPawn) (Variables|EventTags|GetLootSacRescueEvent) true
    (:EventReceived
      (CallFunction|PlayRescueBarks))))

(event Ability|EventFailedToActivatePassiveAbility
  (params (ActorInfo :type Gameplay%20Ability%20Actor%20Info%20Structure))
  ;[0 0 0] Helper events
  (|CallParentFunction ActorInfo))

(event Ability|EventOnEndAbility
  (params (bWasCancelled :type Boolean))
  ;[0 0 0] On end ability
  (|CallParentFunction bWasCancelled))

(fn IsOwnerInSafeZone
  (bind _asfort_player_pawn (Utilities|Casting|CastToFortPlayerPawn (Variables|OwnerInfo|GetOwningPawn)))
  (return (Variables|/Script/FortniteGame.FortPlayerPawn|GetbIsInsideSafeZone _asfort_player_pawn)))
